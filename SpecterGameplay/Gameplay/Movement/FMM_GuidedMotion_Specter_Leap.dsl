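; Guided-motion movement mode for the Specter Leap, overriding the launch velocity calculation to limit how high the player can jump.
; It takes the parent's launch velocity and computes floor distance below the owner's capsule, using the constant-force height-above-ground config as the maximum.
; If the floor is beyond that height or not found, it zeroes vertical velocity and keeps horizontal; otherwise the parent velocity passes through.
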(fn GetLaunchVelocity
  (params (Runtime%20Data :type Fort%20Movement%20Mode%20Guided%20Motion%20Runtime%20Data%20Object%20Reference))
  (bind (_launchvelocity_x _launchvelocity_y) (Math|Vector|BreakVector (Variables|GetLaunchVelocity_X)))
  (bind (_launchvelocity_x_2 _launchvelocity_y_2) (Math|Vector|BreakVector (Variables|GetLaunchVelocity_X)))
  (LocalVariables|SetLaunchVelocity (|CallParentFunction Runtime%20Data))
  (LocalVariables|SetMaxHeight (ScalableFloat|GetValueatLevel (Variables|Config-ConstantForce|GetHeightAboveGround)))
  (bind _floorresult (Pawn|Components|CharacterMovement|ComputeFloorDistance (Variables|FortMovementMode_BaseExtLogic|GetOwnerMovementComponent) (Transformation|GetActorLocation (Variables|FortMovementMode_BaseExtLogic|GetOwnerPawn)) (LocalVariables|GetMaxHeight) (LocalVariables|GetMaxHeight) (Components|Capsule|GetScaledCapsuleRadius (Variables|/Script/Engine.Character|GetCapsuleComponent (Variables|FortMovementMode_BaseExtLogic|GetOwnerPawn)))))
  (if (Utilities|Struct|BreakFindFloorResult _floorresult)
    (if (> (.Distance (Collision|BreakHitResult (.HitResult (Utilities|Struct|BreakFindFloorResult _floorresult)))) (LocalVariables|GetMaxHeight))
      ;[0 0 0 0.5] Don't let the player jump any higher
      (return _launchvelocity_x _launchvelocity_y 0.0)
      (else
        (return (LocalVariables|GetLaunchVelocity))))
    (else
      ;[0 0 0 0.5] We're so high we can't hit the floor at all. Jump no higher
      (return _launchvelocity_x_2 _launchvelocity_y_2 0.0))))
